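# Physics asset for the wolf pet backpack cosmetic, defining collision bodies and joints for its dangling accessories.
# A root box covers the pack, while a box for the pan and capsules for the roll and two feathers cover the dyn_ bones.
# The pan and roll are constrained to the root, and both feathers chain off the pan.
# Swing and twist limits are tighter on the roll than on the pan and feathers.
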PhysicsAsset /BRCosmetics/Accessories/FORT_Backpacks/Backpack_Pets_Wolf/Meshes/WolfPack_Backpack_Physics
bodies {
 dyn_feather_01 [{shapeType=Capsule, length=9.3551, radius=1.5444, center=<7.7458 0.0635 0.149>, rotation=<rot 90 0 0>}]
 dyn_feather_02 [{shapeType=Capsule, length=4.1822, radius=1.3382, center=<4.0339 0.6948 -0.2878>, rotation=<rot 90 0 0>}]
 dyn_pan [{shapeType=Box, extentX=11.7639, extentY=1.2251, extentZ=14.3412, center=<-0.0939 -0.0102 -6.539>}]
 dyn_roll [{shapeType=Capsule, length=15.302, radius=5.9988, center=<0.0007 0.2391 -4.6111>, rotation=<rot 90 0 0>}]
 root [{shapeType=Box, extentX=28.3786, extentY=22.0467, extentZ=26.9827, center=<0 -8.914 -4.2498>}]
}
constraints [
 {bone1Name=dyn_pan, bone2Name=root, swing1LimitDegrees=45, swing2LimitDegrees=45, twistLimitDegrees=15}
 {bone1Name=dyn_feather_01, bone2Name=dyn_pan, swing1LimitDegrees=45, swing2LimitDegrees=45, twistLimitDegrees=15}
 {bone1Name=dyn_feather_02, bone2Name=dyn_pan, swing1LimitDegrees=45, swing2LimitDegrees=45, twistLimitDegrees=15}
 {bone1Name=dyn_roll, bone2Name=root, swing1LimitDegrees=20, swing2LimitDegrees=20, twistLimitDegrees=5}
]
